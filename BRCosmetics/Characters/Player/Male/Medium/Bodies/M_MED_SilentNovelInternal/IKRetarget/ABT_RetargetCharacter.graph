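# Skeleton-agnostic AnimBP that retargets a character's pose from another mesh at runtime through an IK Retargeter.
# Each update copies the retarget profile, then edits op controllers: arm IK, spine stretch and leg IK blend-to-source follow vehicle, parachute, slide, crouch and slope states.
# A pelvis offset is also computed per state, but no visible link feeds it into Pelvis Motion settings.
# Post-retarget weapon bone copy, hand IK retargeting and left-hand Leg IK fix grips.
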AnimBlueprint /BRCosmetics/Characters/Player/Male/Medium/Bodies/M_MED_SilentNovelInternal/IKRetarget/ABT_RetargetCharacter
target_skeleton=null
graphs [
 AnimGraph {nodes [
   n1 = Get IKRetargetProfile() -> (IKRetargetProfile:Retarget Profile Structure)
   n2 = Get IKRetargeterAsset() -> (IKRetargeterAsset:IKRetargeter Object Reference)
   n3 = Copy Bone - Source Bone: hand_attach_l - Target Bone: weapon_r!(
    ComponentPose:Component Space Pose Link Structure=Local To Component(LocalPose=Retarget Pose From Mesh!(IKRetargeterAsset:IKRetargeter Object Reference=@n2, CustomRetargetProfile:Retarget Profile Structure=@n1) -> (Pose:Pose Link Structure) {is=title AnimGraphNode_RetargetPoseFromMesh})
    AlphaCurveName:Name=copyWeaponRtoAttachL
    bCopyTranslation:bool=true
    bCopyRotation:bool=False
    bCopyScale:bool=true
   ) -> (Pose:Component Space Pose Link Structure) {
    is=title AnimGraphNode_CopyBone
    alphaInputType=Curve
    alphaBoolBlend {blendOption=Linear}
    sourceBone {boneName=hand_attach_l}
    targetBone {boneName=weapon_r}
   }
   n4 = Hand IK Retargeting!(ComponentPose:Component Space Pose Link Structure=@n3, Alpha:float=1, HandFKWeight:float=0.5) -> (Pose:Component Space Pose Link Structure) {
    is=title AnimGraphNode_HandIKRetargeting
    alphaBoolBlend {blendOption=Linear}
    iKBonesToMove [{boneName=ik_hand_gun}]
    leftHandFK {boneName=hand_l}
    leftHandIK {boneName=ik_hand_l}
    rightHandFK {boneName=hand_r}
    rightHandIK {boneName=ik_hand_r}
   }
   n5 = Leg IK!(ComponentPose:Component Space Pose Link Structure=@n4, Alpha:float=Property Access!() -> (Value:float=~) {is=title K2Node_PropertyAccess}) -> (Pose:Component Space Pose Link Structure) {
    is=title AnimGraphNode_LegIK
    alphaBoolBlend {blendOption=Linear}
    legsDefinition [{
      bEnableKneeTwistCorrection=false
      bEnableRotationLimit=false
      footBoneForwardAxis=Y
      hingeRotationAxis=Z
      minRotationAngle=15
      numBonesInLimb=2
      twistOffsetCurveName=None
      fKFootBone {boneName=hand_l}
      iKFootBone {boneName=ik_hand_l}
     }]
   }
   Output Pose(Result=Component To Local(ComponentPose=@n5))
  ]}
 Check Slope {nodes [
   Check Slope() {is=title K2Node_FunctionEntry}
   > n1 = Sequence()
   > n2 = Branch(Condition=OR Boolean!(A=Get bIsSliding(), B=Get bIsKneeSliding(), C:bool=Get bIsBaseballSliding(), D:bool=Get bIsSlopeSliding() -> (bIsSlopeSliding:bool)) -> (ReturnValue))
   > Set Current Leg IK Blend to Source(Current Leg IK Blend to Source:double=FInterpTo(
     Current=Get Current Leg IK Blend to Source() -> (Current Leg IK Blend to Source:double)
     Target=1
     DeltaTime=GetWorldDeltaSeconds!() -> (ReturnValue:double)
     InterpSpeed=50
    )) -> (Output_Get:double)
   n3 = Get Current Leg IK Blend to Source()
   > SetIKChainByName!(
    execute
    self
    IKRetargetIKChainsController:IKRetarget IKChains Controller Object Reference=Get IKRetargetIKChainsController() -> (IKRetargetIKChainsController:IKRetarget IKChains Controller Object Reference)
    TargetChainName:Name=RightLegIK
    Enable IK:bool=true
    Blend To Source:double=@n3
    Blend To Source Weights:Vector="1,1,1"
   ) -> (then)
   > SetIKChainByName(
    IKRetargetIKChainsController=Get IKRetargetIKChainsController()
    TargetChainName=LeftLegIK
    Enable IK=true
    Blend To Source=@n3
    Blend To Source Weights="0,1,1"
   )
   n4 = BreakRotator(InRot=Get!() -> (
     SlopeInfoInput_RootSlopeTranslation:Vector=~
     SlopeInfoInput_LeftFootSlopeTranslation:Vector=~
     SlopeInfoInput_RightFootSlopeTranslation:Vector=~
     SlopeInfoInput_SlopeRotation:Rotator=~
    ).SlopeInfoInput_SlopeRotation)
   n5 = MakeVector2D(X=@n4, Y=@n4.Pitch)
   n6 = Branch(execute=@n2.else, Condition="float > float"(A=Vector2D Length!(A:Vector 2D Structure=@n5) -> (ReturnValue:double), B=1))
   n7 = NOT Boolean(A=Get bIsMoving2D())
   > Set Current Leg IK Blend to Source(Current Leg IK Blend to Source=FInterpTo(Current=Get Current Leg IK Blend to Source(), Target=SelectFloat(A=0.6, B=0, bPickA=@n7), DeltaTime=GetWorldDeltaSeconds(), InterpSpeed=30))
   n8 = "float > float"(A=Get Current Leg IK Blend to Source())
   n9 = Get Current Leg IK Blend to Source()
   > SetIKChainByName(
    IKRetargetIKChainsController=Get IKRetargetIKChainsController()
    TargetChainName=RightLegIK
    Enable IK=@n8
    Blend To Source=@n9
    Blend To Source Weights="1,1,1"
   )
   > SetIKChainByName(
    IKRetargetIKChainsController=Get IKRetargetIKChainsController()
    TargetChainName=LeftLegIK
    Enable IK=@n8
    Blend To Source=@n9
    Blend To Source Weights="1,1,1"
   )
   n10 = Get bIsCrouching()
   Set Current Leg IK Blend to Source(
    execute=@n6.else
    Current Leg IK Blend to Source=FInterpTo(Current=Get Current Leg IK Blend to Source(), Target="To Float (Boolean)"(InBool=@n10), DeltaTime=GetWorldDeltaSeconds(), InterpSpeed=50)
   )
   n11 = "float > float"(A=Get Current Leg IK Blend to Source())
   n12 = Get Current Leg IK Blend to Source()
   > SetIKChainByName(
    IKRetargetIKChainsController=Get IKRetargetIKChainsController()
    TargetChainName=RightLegIK
    Enable IK=@n11
    Blend To Source=@n12
    Blend To Source Weights="0,0,1"
   )
   > SetIKChainByName(
    IKRetargetIKChainsController=Get IKRetargetIKChainsController()
    TargetChainName=LeftLegIK
    Enable IK=@n11
    Blend To Source=@n12
    Blend To Source Weights="0,0,1"
   )
   Return Node(execute=@n1.then_1)
  ]}
 EventGraph {nodes [
   n1 = Event BlueprintUpdateAnimation()
   > n2 = Set Delta Time X(Delta Time X=@n1.DeltaTimeX)
   > Parent: BlueprintUpdateAnimation!(execute, DeltaTimeX:float=@n2) -> (then) {is=title K2Node_CallParentFunction}
   > n3 = CopyRetargetProfileFromRetargetAsset!(execute, InRetargetAsset:IKRetargeter Object Reference=Get IKRetargeterAsset()) -> (then, ReturnValue:Retarget Profile Structure)
   > Set IKRetargetProfile(IKRetargetProfile:Retarget Profile Structure=@n3) -> (Output_Get:Retarget Profile Structure)
   > n4 = Sequence!(execute) -> (then_0, then_1, then_2:Exec, then_3:Exec)
   > n5 = Cast To IKRetargetPelvisMotionController!(
    execute
    Object=GetOpControllerFromRetargetProfile!(InRetargetProfile:"Retarget Profile Structure (by ref)"=Get IKRetargetProfile(), InRetargetOpName:Name=Pelvis Motion) -> (ReturnValue:IKRetarget Op Controller Base Object Reference)
   ) -> (then, CastFailed, AsIKRetarget Pelvis Motion Controller:IKRetarget Pelvis Motion Controller Object Reference) {is=title K2Node_DynamicCast}
   > Set IKRetargetPelvisMotionController(IKRetargetPelvisMotionController:IKRetarget Pelvis Motion Controller Object Reference=@n5) -> (Output_Get:IKRetarget Pelvis Motion Controller Object Reference)
   n6 = Cast To IKRetargetIKChainsController!(
    execute=@n4.then_1
    Object=GetOpControllerFromRetargetProfile(InRetargetProfile=Get IKRetargetProfile(), InRetargetOpName=Retarget Limb IK Goals)
   ) -> (then, CastFailed, AsIKRetarget IKChains Controller:IKRetarget IKChains Controller Object Reference) {is=title K2Node_DynamicCast}
   > Set IKRetargetIKChainsController(IKRetargetIKChainsController:IKRetarget IKChains Controller Object Reference=@n6) -> (Output_Get:IKRetarget IKChains Controller Object Reference)
   Set LeftArmIK(execute=@n4.then_3, LeftArmIK:bool=false) -> (Output_Get:bool)
   > Set RightArmIK(RightArmIK:bool=false) -> (Output_Get:bool)
   > Set SpineStretch(SpineStretch:bool=false) -> (Output_Get:bool)
   > n7 = Sequence!(execute) -> (then_0, then_1, then_2, then_3, then_4:Exec, then_5:Exec, then_6:Exec)
   Is Valid(
    exec=@n7.then_3
    InputObject=Get IKRetargetPelvisMotionController() -> (IKRetargetPelvisMotionController:IKRetarget Pelvis Motion Controller Object Reference)
   )
   > n8 = GetSettings!(execute, self:IKRetarget Pelvis Motion Controller Object Reference=Get IKRetargetPelvisMotionController()) -> (then, ReturnValue:Pelvis Motion Settings Structure)
   > n9 = Set members in Pelvis Motion Settings!(execute:Exec, StructRef:"Pelvis Motion Settings Structure (by ref)"=@n8) -> (then:Exec, StructOut:"Pelvis Motion Settings Structure (by ref)") {is=title K2Node_SetFieldsInStruct}
   > SetSettings!(
    execute
    self:IKRetarget Pelvis Motion Controller Object Reference=Get IKRetargetPelvisMotionController()
    InSettings:Pelvis Motion Settings Structure=@n9
   ) -> (then)
   n10 = Get LeftArmIK() -> (LeftArmIK:bool)
   SetIKChainByName(
    execute=@n7.then_4
    IKRetargetIKChainsController=Get IKRetargetIKChainsController()
    TargetChainName=LeftArm
    Enable IK=@n10
    Blend To Source="To Float (Boolean)"(InBool=@n10)
    Blend To Source Weights="1,1,1"
   )
   n11 = Get RightArmIK() -> (RightArmIK:bool)
   > SetIKChainByName(
    IKRetargetIKChainsController=Get IKRetargetIKChainsController()
    TargetChainName=RightArm
    Enable IK=@n11
    Blend To Source="To Float (Boolean)"(InBool=@n11)
    Blend To Source Weights="1,1,1"
   )
   n12 = Get SpineStretch() -> (SpineStretch:bool)
   SetStretchChainByName!(
    execute=@n7.then_5
    self
    IkRetargetStretchChainController:IKRetarget Stretch Chain Controller Object Reference=Get IKRetargetStretchChainController() -> (IKRetargetStretchChainController:IKRetarget Stretch Chain Controller Object Reference)
    ChainName:Name=Spine
    EnableStretch:bool=@n12
    Match Source Length:double="To Float (Boolean)"(InBool=@n12)
   ) -> (then)
   Check Slope(execute=@n7.then_6) {is=title K2Node_CallFunction}
   n13 = Get VehicleSeatIndex() -> (VehicleSeatIndex:int=0)
   n14 = OR Boolean!(
    A=Get bParachuteOpenedAndStandingUpright() -> (bParachuteOpenedAndStandingUpright:bool)
    B=AND Boolean(A="Equal (Integer)"(A=@n13), B=Get bIsInVehicle() -> (bIsInVehicle:bool))
   ) -> (ReturnValue)
   Set LeftArmIK(execute=@n7.then_0, LeftArmIK=@n14)
   > Set RightArmIK(RightArmIK=@n14)
   n15 = Branch(execute=@n7.then_1, Condition=Get bIsInVehicle())
   > Set CurrentPelvisTranslationOffset(CurrentPelvisTranslationOffset:Vector=Get In Vehicle Pelvis Translation Offset() -> (In Vehicle Pelvis Translation Offset:Vector)) -> (Output_Get:Vector)
   n16 = Branch(execute=@n15.else, Condition=Get bIsCrouching())
   > n17 = Branch(Condition=OR Boolean!(A=Get bIsSliding(), B=Get bIsKneeSliding(), C:bool=Get bIsBaseballSliding(), D:bool=Get bIsSlopeSliding()) -> (ReturnValue))
   > Interpolate Pelvis Offset!(execute, self, Target:Vector=Get Sliding Pelvis Translation Offset() -> (Sliding Pelvis Translation Offset:Vector), InterpSpeed:float=10) -> (then, Output:Vector)
   Interpolate Pelvis Offset(
    execute=@n17.else
    Target=SelectVector(
     A=Get Crouching Moving Pelvis Translation Offset() -> (Crouching Moving Pelvis Translation Offset:Vector)
     B=Get Crouching Idle Pelvis Translation Offset() -> (Crouching Idle Pelvis Translation Offset:Vector)
     bPickA=Get bIsMoving2D()
    )
    InterpSpeed=10
   )
   n18 = BreakRotator(InRot=Get() -> (
     SlopeInfoInput_RootSlopeTranslation=~
     SlopeInfoInput_LeftFootSlopeTranslation=~
     SlopeInfoInput_RightFootSlopeTranslation=~
     SlopeInfoInput_SlopeRotation=~
    ).SlopeInfoInput_SlopeRotation)
   n19 = Vector2D Length(A=MakeVector2D(X=@n18, Y=@n18.Pitch))
   Interpolate Pelvis Offset(
    execute=@n16.else
    Target=SelectVector(
     A=Get On Slope Pelvis Translation Offset() -> (On Slope Pelvis Translation Offset:Vector)
     B=Get Standing Pelvis Translation Offset() -> (Standing Pelvis Translation Offset:Vector)
     bPickA="float > float"(A=@n19, B=1)
    )
    InterpSpeed=20
   )
   Branch(execute=@n7.then_2, Condition=Get bIsInVehicle())
   > Set SpineStretch(SpineStretch=true)
   n20 = Cast To IKRetargetStretchChainController!(execute=@n4.then_2, Object=GetOpControllerFromRetargetProfile(InRetargetProfile=Get IKRetargetProfile(), InRetargetOpName=Stretch Chains)) -> (then, CastFailed, AsIKRetarget Stretch Chain Controller:IKRetarget Stretch Chain Controller Object Reference) {is=title K2Node_DynamicCast}
   > Set IKRetargetStretchChainController(IKRetargetStretchChainController:IKRetarget Stretch Chain Controller Object Reference=@n20) -> (Output_Get:IKRetarget Stretch Chain Controller Object Reference)
   Get CurrentPelvisTranslationOffset() -> (CurrentPelvisTranslationOffset:Vector)
  ]}
 FindIKChainByName {nodes [
   n1 = FindIKChainByName!() -> (then, Array:Array of Retarget IKChain Settings Structures, ChainName:Name) {is=title K2Node_FunctionEntry}
   > n2 = For Each Loop(Array:Array of Retarget IKChain Settings Structures=@n1) -> (Array Element:Retarget IKChain Settings Structure)
   Return Node!(execute=@n2.Completed, Array Element:Retarget IKChain Settings Structure, Found Index:int=-1)
   > Branch(Condition="Equal (Name)"(
     A=Break Retarget IKChain Settings!(RetargetIKChainSettings:"Retarget IKChain Settings Structure (by ref)"=@n2) -> (TargetChainName:Name, ApplyPelvisOffsetToSourceGoals:bool=~) {is=title K2Node_BreakStruct}
     B=Get ChainName() -> (ChainName:Name=None)
    ))
   > Return Node(Array Element=@n2, Found Index=@n2.Array Index)
  ]}
 FindStretchChainByName {nodes [
   n1 = FindStretchChainByName!() -> (then, Array:Array of Retarget Stretch Chain Settings Structures, ChainName:Name) {is=title K2Node_FunctionEntry}
   > n2 = For Each Loop(Array:Array of Retarget Stretch Chain Settings Structures=@n1) -> (Array Element:Retarget Stretch Chain Settings Structure)
   Return Node(execute=@n2.Completed, Array Element:Retarget Stretch Chain Settings Structure=@n2, Found Index=-1)
   > Branch(Condition="Equal (Name)"(
     A=Break Retarget Stretch Chain Settings!(RetargetStretchChainSettings:"Retarget Stretch Chain Settings Structure (by ref)"=@n2) -> (TargetChainName:Name, bEnabled:bool=~, MatchSourceLength:double=~, ScaleChainLength:double=~) {is=title K2Node_BreakStruct}
     B=Get ChainName() -> (ChainName=None)
    ))
   > Return Node(Array Element=@n2, Found Index=@n2.Array Index)
  ]}
 Interpolate Pelvis Offset {nodes [
   n1 = Interpolate Pelvis Offset!() -> (then, Target:Vector=~, InterpSpeed:float=~) {is=title K2Node_FunctionEntry}
   > n2 = Set CurrentPelvisTranslationOffset(CurrentPelvisTranslationOffset=VInterpTo!(
     Current:Vector=Get CurrentPelvisTranslationOffset()
     Target:Vector=@n1
     DeltaTime:float=Get Delta Time X()
     InterpSpeed:float=@n1.InterpSpeed
    ) -> (ReturnValue:Vector))
   > Return Node!(execute, Output:Vector=@n2)
  ]}
 SetIKChainByName {nodes [
   n1 = SetIKChainByName!() -> (
    then
    IKRetargetIKChainsController:IKRetarget IKChains Controller Object Reference
    TargetChainName:Name
    Enable IK:bool=~
    Blend To Source:double=~
    Blend To Source Weights:Vector="1,1,1"
   ) {is=title K2Node_FunctionEntry}
   > Is Valid(InputObject=@n1)
   > n2 = GetSettings(self:IKRetarget IKChains Controller Object Reference=@n1) -> (ReturnValue:Retarget IK Goals Settings Structure)
   n3 = Break Retarget IK Goals Settings!(IKRetargetIKChainsOpSettings:"Retarget IK Goals Settings Structure (by ref)"=@n2) -> (ChainsToRetarget:Array of Retarget IKChain Settings Structures, bEnabled:bool=~, LODThreshold:int) {is=title K2Node_BreakStruct}
   > Set IKChains(IKChains:Array of Retarget IKChain Settings Structures=@n3) -> (Output_Get:Array of Retarget IKChain Settings Structures)
   > n4 = FindIKChainByName!(execute, self, Array:Array of Retarget IKChain Settings Structures=@n3, ChainName:Name=Get TargetChainName() -> (TargetChainName:Name=None)) -> (then, Array Element:Retarget IKChain Settings Structure, Found Index:int=0) {is=title K2Node_CallFunction}
   > Branch(Condition="Not Equal (Integer)"(A=@n4.Found Index, B=-1))
   > n5 = Set members in Retarget IKChain Settings!(
    execute:Exec
    StructRef:"Retarget IKChain Settings Structure (by ref)"=@n4
    EnableIK:bool=Get Enable IK() -> (Enable IK:bool)
    BlendToSource:double=Get Blend To Source() -> (Blend To Source:double)
    BlendToSourceWeights:Vector=Get Blend To Source Weights() -> (Blend To Source Weights:Vector)
   ) -> (then:Exec, StructOut:"Retarget IKChain Settings Structure (by ref)") {is=title K2Node_SetFieldsInStruct}
   > Set Array Elem!(
    execute
    TargetArray:Array of Retarget IKChain Settings Structures=Get IKChains() -> (IKChains:Array of Retarget IKChain Settings Structures)
    Index:int=@n4.Found Index
    Item:"Retarget IKChain Settings Structure (by ref)"=@n5
    bSizeToFit:bool=false
   ) -> (then)
   > n6 = Set members in Retarget IK Goals Settings!(
    execute:Exec
    StructRef:"Retarget IK Goals Settings Structure (by ref)"=@n2
    ChainsToRetarget:Array of Retarget IKChain Settings Structures=Get IKChains()
   ) -> (then:Exec, StructOut:"Retarget IK Goals Settings Structure (by ref)") {is=title K2Node_SetFieldsInStruct}
   > SetSettings(self:IKRetarget IKChains Controller Object Reference=Get IKRetargetIKChainsController(), InSettings:Retarget IK Goals Settings Structure=@n6)
  ]}
 SetStretchChainByName {nodes [
   n1 = SetStretchChainByName!() -> (
    then
    IkRetargetStretchChainController:IKRetarget Stretch Chain Controller Object Reference
    ChainName:Name
    EnableStretch:bool=~
    Match Source Length:double=~
   ) {is=title K2Node_FunctionEntry}
   > Is Valid(InputObject=@n1)
   > n2 = GetSettings(self:IKRetarget Stretch Chain Controller Object Reference=@n1) -> (ReturnValue:Stretch Chain Settings Structure)
   > n3 = Set ChainsToStretch(ChainsToStretch:Array of Retarget Stretch Chain Settings Structures=Break Stretch Chain Settings!(IKRetargetStretchChainOpSettings:"Stretch Chain Settings Structure (by ref)"=@n2) -> (
     IKRigAsset:IKRig Definition Object Reference
     ChainsToStretch:Array of Retarget Stretch Chain Settings Structures
     bEnabled:bool=~
     LODThreshold:int
    ) {is=title K2Node_BreakStruct}.ChainsToStretch) -> (Output_Get:Array of Retarget Stretch Chain Settings Structures)
   > n4 = FindStretchChainByName!(execute, self, Array:Array of Retarget Stretch Chain Settings Structures=@n3, ChainName:Name=Get ChainName() -> (ChainName=None)) -> (then, Array Element:Retarget Stretch Chain Settings Structure, Found Index:int=0) {is=title K2Node_CallFunction}
   > Branch(Condition="Not Equal (Integer)"(A=@n4.Found Index, B=-1))
   > n5 = Set members in Retarget Stretch Chain Settings!(
    execute:Exec
    StructRef:"Retarget Stretch Chain Settings Structure (by ref)"=@n4
    bEnabled:bool=Get EnableStretch() -> (EnableStretch:bool)
    MatchSourceLength:double=Get Match Source Length() -> (Match Source Length:double)
   ) -> (then:Exec, StructOut:"Retarget Stretch Chain Settings Structure (by ref)") {is=title K2Node_SetFieldsInStruct}
   > Set Array Elem(
    TargetArray:Array of Retarget Stretch Chain Settings Structures=Get ChainsToStretch() -> (ChainsToStretch:Array of Retarget Stretch Chain Settings Structures)
    Index=@n4.Found Index
    Item:"Retarget Stretch Chain Settings Structure (by ref)"=@n5
    bSizeToFit=false
   )
   > n6 = Set members in Stretch Chain Settings!(
    execute:Exec
    StructRef:"Stretch Chain Settings Structure (by ref)"=@n2
    ChainsToStretch:Array of Retarget Stretch Chain Settings Structures=Get ChainsToStretch()
   ) -> (then:Exec, StructOut:"Stretch Chain Settings Structure (by ref)") {is=title K2Node_SetFieldsInStruct}
   > SetSettings(
    self:IKRetarget Stretch Chain Controller Object Reference=Get IkRetargetStretchChainController() -> (IkRetargetStretchChainController:IKRetarget Stretch Chain Controller Object Reference)
    InSettings:Stretch Chain Settings Structure=@n6
   )
  ]}
]
variables [
 Crouching Idle Pelvis Translation Offset {type=Vector, default=<0 0 20>}
 Crouching Moving Pelvis Translation Offset {type=Vector, default=<0 0 5>}
 Current Leg IK Blend to Source {type=number}
 CurrentPelvisTranslationOffset {type=Vector}
 Delta Time X {type=number}
 IKRetargeterAsset {type=IKRetargeter, default=null}
 IKRetargetIKChainsController {type=IKRetargetIKChainsController, default=null}
 IKRetargetPelvisMotionController {type=IKRetargetPelvisMotionController, default=null}
 IKRetargetProfile {
  type=RetargetProfile
  default {
   bApplyChainSettings=false
   bApplyGlobalSettings=false
   bApplyRootSettings=false
   bApplySourceRetargetPose=false
   bApplyTargetRetargetPose=false
   bForceAllIKOff=false
   sourceRetargetPoseName=None
   targetRetargetPoseName=None
   chainSettings {}
   globalSettings {
    bCopyBasePose=false
    bEnableFK=true
    bEnableIK=true
    bEnablePost=true
    bEnableRoot=true
    bWarping=false
    copyBasePoseRoot=None
    directionChain=None
    directionSource=Goals
    forwardDirection=Y
    sidewaysOffset=0
    sourceScaleFactor=1
    warpForwards=1
    warpSplay=1
   }
   retargetOpProfiles []
   rootSettings {
    affectIKHorizontal=1
    affectIKVertical=0
    blendToSource=0
    rotationAlpha=1
    scaleHorizontal=1
    scaleVertical=1
    translationAlpha=1
    blendToSourceWeights=<1 1 1>
    rotationOffset=<rot 0 0 0>
    translationOffset=<0 0 0>
   }
  }
 }
 IKRetargetStretchChainController {type=IKRetargetStretchChainController, default=null}
 In Vehicle Pelvis Translation Offset {type=Vector, default=<0 0 10>}
 LeftArmIK {type=boolean}
 On Slope Pelvis Translation Offset {type=Vector, default=<0 0 2>}
 RightArmIK {type=boolean}
 Sliding Pelvis Translation Offset {type=Vector, default=<0 0 18>}
 SpineStretch {type=boolean, default=true}
 Standing Pelvis Translation Offset {type=Vector}
]
