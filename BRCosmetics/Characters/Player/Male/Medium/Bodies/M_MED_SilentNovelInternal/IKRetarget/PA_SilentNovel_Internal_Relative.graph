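# Physics asset for the SilentNovel Internal medium male player body, kept with its IK retarget assets and named as the "Relative" variant.
# It approximates pelvis, upper spine, neck, head, arms, hands, legs and feet with one capsule per bone, mirrored left to right.
# Constraints chain these bodies into a tree rooted at the pelvis, with uniform swing and twist limit angles.
# Both neck joints and both hip joints explicitly set all angular motion free.
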PhysicsAsset /BRCosmetics/Characters/Player/Male/Medium/Bodies/M_MED_SilentNovelInternal/IKRetarget/PA_SilentNovel_Internal_Relative
bodies {
 calf_l [{shapeType=Capsule, length=12.6159, radius=13.7146, shapeName=calf_l_capsule, center=<-16.1287 -0.0728 0>, rotation=<rot -90 -0.0003 0>}]
 calf_r [{shapeType=Capsule, length=12.6159, radius=13.7146, shapeName=calf_r_capsule, center=<16.1287 0.0728 0>, rotation=<rot -90 179.9997 0>}]
 foot_l [{shapeType=Capsule, length=13.2445, radius=11.2401, shapeName=foot_l_capsule, center=<-7.3067 7.2822 0.9327>, rotation=<rot 0 0 90>}]
 foot_r [{shapeType=Capsule, length=13.2445, radius=11.2401, shapeName=foot_r_capsule, center=<7.3067 -7.2822 -0.9327>, rotation=<rot 0 0 -90>}]
 hand_l [{shapeType=Capsule, length=7.6713, radius=10.0772, shapeName=hand_l_capsule, center=<11.8531 -2.2896 0>, rotation=<rot 90 -9.9998 0>}]
 hand_r [{shapeType=Capsule, length=7.6713, radius=10.0772, shapeName=hand_r_capsule, center=<-11.8531 2.2896 0>, rotation=<rot 90 170.0002 0>}]
 head [{shapeType=Capsule, length=15.8312, radius=19.2588, shapeName=head_capsule, center=<22.2673 6.3822 0>, rotation=<rot 0 49.9984 -90>}]
 lowerarm_l [{shapeType=Capsule, length=12.9472, radius=7.5528, shapeName=lowerarm_l_capsule, center=<9.0972 0.3092 0>, rotation=<rot 90 180 -0.0008>}]
 lowerarm_r [{shapeType=Capsule, length=12.9472, radius=7.5528, shapeName=lowerarm_r_capsule, center=<-9.0972 -0.3092 0>, rotation=<rot 90 0.0008 0>}]
 neck_01 [{shapeType=Capsule, length=10, radius=12, shapeName=neck_01_capsule}]
 neck_02 [{shapeType=Capsule, length=4, radius=11.069, shapeName=neck_02_capsule, center=<-1.7834 1.9657 0>, rotation=<rot -90 -0.0003 0>}]
 pelvis [{shapeType=Capsule, length=4.1915, radius=29.5445, center=<4.9349 -1.5411 -0.0172>, rotation=<rot 89.2695 -65.8431 24.1551>}]
 spine_03 [{shapeType=Capsule, length=1.411, radius=16.9961, shapeName=Spine_03_capsule, center=<6.2202 0.5311 0>}]
 spine_05 [{shapeType=Capsule, length=7.2001, radius=14, shapeName=spine_05_capsule, center=<-2.4431 0 0>, rotation=<rot 0 -0.0012 0>}]
 thigh_l [{shapeType=Capsule, length=11.7729, radius=12.5879, shapeName=thigh_l_capsule, center=<-13.5908 0.0005 0>, rotation=<rot -90 -0.0032 0>}]
 thigh_r [{shapeType=Capsule, length=11.7729, radius=12.5879, shapeName=thigh_r_capsule, center=<13.5908 -0.0005 0>, rotation=<rot -90 179.9968 0>}]
 upperarm_l [{shapeType=Capsule, length=16.2401, radius=9.2599, shapeName=upperarm_l_capsule, center=<11.1091 0.0001 0>, rotation=<rot 0.0001 90.0011 -90>}]
 upperarm_r [{
   shapeType=Capsule
   length=16.2401
   radius=9.2599
   shapeName=upperarm_r_capsule
   center=<-11.1091 -0.0001 0>
   rotation=<rot 0.0001 90.0011 90>
  }]
}
constraints [
 {bone1Name=calf_l, bone2Name=thigh_l, swing1LimitDegrees=45, swing2LimitDegrees=45, twistLimitDegrees=45}
 {bone1Name=foot_l, bone2Name=calf_l, swing1LimitDegrees=45, swing2LimitDegrees=45, twistLimitDegrees=45}
 {bone1Name=calf_r, bone2Name=thigh_r, swing1LimitDegrees=45, swing2LimitDegrees=45, twistLimitDegrees=45}
 {bone1Name=foot_r, bone2Name=calf_r, swing1LimitDegrees=45, swing2LimitDegrees=45, twistLimitDegrees=45}
 {bone1Name=spine_03, bone2Name=pelvis, swing1LimitDegrees=45, swing2LimitDegrees=45, twistLimitDegrees=45}
 {bone1Name=spine_05, bone2Name=spine_03, swing1LimitDegrees=45, swing2LimitDegrees=45, twistLimitDegrees=45}
 {bone1Name=head, bone2Name=neck_02, swing1LimitDegrees=45, swing2LimitDegrees=45, twistLimitDegrees=45}
 {bone1Name=upperarm_r, bone2Name=spine_05, swing1LimitDegrees=45, swing2LimitDegrees=45, twistLimitDegrees=45}
 {bone1Name=lowerarm_r, bone2Name=upperarm_r, swing1LimitDegrees=45, swing2LimitDegrees=45, twistLimitDegrees=45}
 {bone1Name=hand_r, bone2Name=lowerarm_r, swing1LimitDegrees=45, swing2LimitDegrees=45, twistLimitDegrees=45}
 {bone1Name=upperarm_l, bone2Name=spine_05, swing1LimitDegrees=45, swing2LimitDegrees=45, twistLimitDegrees=45}
 {bone1Name=lowerarm_l, bone2Name=upperarm_l, swing1LimitDegrees=45, swing2LimitDegrees=45, twistLimitDegrees=45}
 {bone1Name=hand_l, bone2Name=lowerarm_l, swing1LimitDegrees=45, swing2LimitDegrees=45, twistLimitDegrees=45}
 {
  bone1Name=neck_02
  bone2Name=neck_01
  swing1LimitDegrees=45
  swing1Motion=Free
  swing2LimitDegrees=45
  swing2Motion=Free
  twistLimitDegrees=45
  twistMotion=Free
 }
 {
  bone1Name=thigh_l
  bone2Name=pelvis
  swing1LimitDegrees=45
  swing1Motion=Free
  swing2LimitDegrees=45
  swing2Motion=Free
  twistLimitDegrees=45
  twistMotion=Free
 }
 {
  bone1Name=thigh_r
  bone2Name=pelvis
  swing1LimitDegrees=45
  swing1Motion=Free
  swing2LimitDegrees=45
  swing2Motion=Free
  twistLimitDegrees=45
  twistMotion=Free
 }
 {
  bone1Name=neck_01
  bone2Name=spine_05
  swing1LimitDegrees=45
  swing1Motion=Free
  swing2LimitDegrees=45
  swing2Motion=Free
  twistLimitDegrees=45
  twistMotion=Free
 }
]
