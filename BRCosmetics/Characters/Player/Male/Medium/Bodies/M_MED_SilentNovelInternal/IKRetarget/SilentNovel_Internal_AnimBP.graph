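# Animation blueprint for the SilentNovel Internal player body, retargeting the pose from a mesh through an IK Retargeter, then applying hand IK retargeting and left-hand Leg IK.
# Each update copies the retarget profile and fetches its pelvis-motion and limb-IK op controllers.
# A helper function enables arm-chain IK only in vehicle seat zero.
# Vehicle and crouch states set a pelvis offset, but pelvis settings are written back without it.
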AnimBlueprint /BRCosmetics/Characters/Player/Male/Medium/Bodies/M_MED_SilentNovelInternal/IKRetarget/SilentNovel_Internal_AnimBP
target_skeleton=/BRCosmetics/Characters/Player/Male/Medium/Bodies/M_MED_SilentNovelInternal/Meshes/M_MED_SilentNovel_Internal_Skeleton
graphs [
 AnimGraph {nodes [
   n1 = Get IKRetargetProfile() -> (IKRetargetProfile:Retarget Profile Structure)
   n2 = Get IKRetargeterAsset() -> (IKRetargeterAsset:IKRetargeter Object Reference)
   n3 = Hand IK Retargeting!(
    ComponentPose:Component Space Pose Link Structure=Local To Component(LocalPose=Retarget Pose From Mesh!(IKRetargeterAsset:IKRetargeter Object Reference=@n2, CustomRetargetProfile:Retarget Profile Structure=@n1) -> (Pose:Pose Link Structure) {is=title AnimGraphNode_RetargetPoseFromMesh})
    Alpha:float=1
    HandFKWeight:float=0.5
   ) -> (Pose:Component Space Pose Link Structure) {
    is=title AnimGraphNode_HandIKRetargeting
    alphaBoolBlend {blendOption=Linear}
    iKBonesToMove [{boneName=ik_hand_gun}]
    leftHandFK {boneName=hand_l}
    leftHandIK {boneName=ik_hand_l}
    rightHandFK {boneName=hand_r}
    rightHandIK {boneName=ik_hand_r}
   }
   n4 = Leg IK!(ComponentPose:Component Space Pose Link Structure=@n3, Alpha:float=Property Access!() -> (Value:float=~) {is=title K2Node_PropertyAccess}) -> (Pose:Component Space Pose Link Structure) {
    is=title AnimGraphNode_LegIK
    alphaBoolBlend {blendOption=Linear}
    legsDefinition [{
      bEnableKneeTwistCorrection=false
      bEnableRotationLimit=false
      footBoneForwardAxis=Y
      hingeRotationAxis=Z
      minRotationAngle=15
      numBonesInLimb=2
      twistOffsetCurveName=None
      fKFootBone {boneName=hand_l}
      iKFootBone {boneName=ik_hand_l}
     }]
   }
   Output Pose(Result=Component To Local(ComponentPose=@n4))
  ]}
 EventGraph {nodes [
   n1 = Event BlueprintUpdateAnimation()
   > Parent: BlueprintUpdateAnimation!(execute, DeltaTimeX:float=@n1.DeltaTimeX) -> (then) {is=title K2Node_CallParentFunction}
   > n2 = CopyRetargetProfileFromRetargetAsset!(execute, InRetargetAsset:IKRetargeter Object Reference=Get IKRetargeterAsset()) -> (then, ReturnValue:Retarget Profile Structure)
   > Set IKRetargetProfile(IKRetargetProfile:Retarget Profile Structure=@n2) -> (Output_Get:Retarget Profile Structure)
   > n3 = Sequence!(execute) -> (then_0, then_1, then_2:Exec)
   > n4 = Cast To IKRetargetPelvisMotionController!(
    execute
    Object=GetOpControllerFromRetargetProfile!(InRetargetProfile:"Retarget Profile Structure (by ref)"=Get IKRetargetProfile(), InRetargetOpName:Name=Pelvis Motion) -> (ReturnValue:IKRetarget Op Controller Base Object Reference)
   ) -> (then, CastFailed, AsIKRetarget Pelvis Motion Controller:IKRetarget Pelvis Motion Controller Object Reference) {is=title K2Node_DynamicCast}
   > Set IKRetargetPelvisMotionController(IKRetargetPelvisMotionController:IKRetarget Pelvis Motion Controller Object Reference=@n4) -> (Output_Get:IKRetarget Pelvis Motion Controller Object Reference)
   n5 = Cast To IKRetargetIKChainsController!(
    execute=@n3.then_1
    Object=GetOpControllerFromRetargetProfile(InRetargetProfile=Get IKRetargetProfile(), InRetargetOpName=Retarget Limb IK Goals)
   ) -> (then, CastFailed, AsIKRetarget IKChains Controller:IKRetarget IKChains Controller Object Reference) {is=title K2Node_DynamicCast}
   > Set IKRetargetIKChainsController(IKRetargetIKChainsController:IKRetarget IKChains Controller Object Reference=@n5) -> (Output_Get:IKRetarget IKChains Controller Object Reference)
   Set PelvisTranslationOffset(execute=@n3.then_2, PelvisTranslationOffset:Vector="0, 0, 0") -> (Output_Get:Vector)
   > Set LeftArmIK(LeftArmIK:bool=false) -> (Output_Get:bool)
   > Set RightArmIK(RightArmIK:bool=false) -> (Output_Get:bool)
   > n6 = Sequence!(execute) -> (then_0, then_1, then_2, then_3:Exec)
   n7 = Sequence!(execute=@n6.then_1) -> (then_0, then_1, then_2)
   > Branch(Condition=Get bIsInVehicle() -> (bIsInVehicle:bool))
   > Set PelvisTranslationOffset(PelvisTranslationOffset="0,0,10")
   Branch(execute=@n7.then_2, Condition=Get bIsCrouching())
   > Set PelvisTranslationOffset(PelvisTranslationOffset="0,0,50")
   Is Valid(
    exec=@n6.then_2
    InputObject=Get IKRetargetPelvisMotionController() -> (IKRetargetPelvisMotionController:IKRetarget Pelvis Motion Controller Object Reference)
   )
   > n8 = GetSettings!(execute, self:IKRetarget Pelvis Motion Controller Object Reference=Get IKRetargetPelvisMotionController()) -> (then, ReturnValue:Pelvis Motion Settings Structure)
   > n9 = Set members in Pelvis Motion Settings!(execute:Exec, StructRef:"Pelvis Motion Settings Structure (by ref)"=@n8) -> (then:Exec, StructOut:"Pelvis Motion Settings Structure (by ref)") {is=title K2Node_SetFieldsInStruct}
   > SetSettings!(
    execute
    self:IKRetarget Pelvis Motion Controller Object Reference=Get IKRetargetPelvisMotionController()
    InSettings:Pelvis Motion Settings Structure=@n9
   ) -> (then)
   SetIKChainByName!(
    execute=@n6.then_3
    self
    IKRetargetIKChainsController:IKRetarget IKChains Controller Object Reference=Get IKRetargetIKChainsController() -> (IKRetargetIKChainsController:IKRetarget IKChains Controller Object Reference)
    TargetChainName:Name=LeftArm
    Enable IK:bool=Get LeftArmIK() -> (LeftArmIK:bool)
   ) -> (then)
   > SetIKChainByName(IKRetargetIKChainsController=Get IKRetargetIKChainsController(), TargetChainName=RightArm, Enable IK=Get RightArmIK() -> (RightArmIK:bool))
   n10 = AND Boolean(A="Equal (Integer)"(A=Get VehicleSeatIndex() -> (VehicleSeatIndex:int=0)), B=Get bIsInVehicle())
   Set LeftArmIK(execute=@n6.then_0, LeftArmIK=@n10)
   > Set RightArmIK(RightArmIK=@n10)
   Get PelvisTranslationOffset() -> (PelvisTranslationOffset:Vector)
  ]}
 SetIKChainByName {nodes [
   n1 = SetIKChainByName!() -> (then, IKRetargetIKChainsController:IKRetarget IKChains Controller Object Reference, TargetChainName:Name, Enable IK:bool=~) {is=title K2Node_FunctionEntry}
   > Is Valid(InputObject=@n1)
   > n2 = GetSettings(self:IKRetarget IKChains Controller Object Reference=@n1) -> (ReturnValue:Retarget IK Goals Settings Structure)
   > n3 = Set IKChains(IKChains:Array of Retarget IKChain Settings Structures=Break Retarget IK Goals Settings!(IKRetargetIKChainsOpSettings:"Retarget IK Goals Settings Structure (by ref)"=@n2) -> (ChainsToRetarget:Array of Retarget IKChain Settings Structures, bEnabled:bool=~, LODThreshold:int) {is=title K2Node_BreakStruct}) -> (Output_Get:Array of Retarget IKChain Settings Structures)
   > n4 = For Each Loop(Array:Array of Retarget IKChain Settings Structures=@n3) -> (Array Element:Retarget IKChain Settings Structure)
   n5 = Set members in Retarget IK Goals Settings!(
    execute:Exec=@n4.Completed
    StructRef:"Retarget IK Goals Settings Structure (by ref)"=@n2
    ChainsToRetarget:Array of Retarget IKChain Settings Structures=Get IKChains() -> (IKChains:Array of Retarget IKChain Settings Structures)
   ) -> (then:Exec, StructOut:"Retarget IK Goals Settings Structure (by ref)") {is=title K2Node_SetFieldsInStruct}
   > SetSettings(self:IKRetarget IKChains Controller Object Reference=Get IKRetargetIKChainsController(), InSettings:Retarget IK Goals Settings Structure=@n5)
   > Return Node()
   Branch(
    execute=@n4.LoopBody
    Condition="Equal (Name)"(
     A=Break Retarget IKChain Settings!(RetargetIKChainSettings:"Retarget IKChain Settings Structure (by ref)"=@n4) -> (TargetChainName:Name, ApplyPelvisOffsetToSourceGoals:bool=~) {is=title K2Node_BreakStruct}
     B=Get TargetChainName() -> (TargetChainName:Name=None)
    )
   )
   n6 = Get Enable IK() -> (Enable IK:bool)
   > n7 = Set members in Retarget IKChain Settings!(
    execute:Exec
    StructRef:"Retarget IKChain Settings Structure (by ref)"=@n4
    EnableIK:bool=@n6
    BlendToSource:double="To Float (Boolean)"(InBool=@n6)
   ) -> (then:Exec, StructOut:"Retarget IKChain Settings Structure (by ref)") {is=title K2Node_SetFieldsInStruct}
   > Set Array Elem!(
    execute
    TargetArray:Array of Retarget IKChain Settings Structures=Get IKChains()
    Index:int=@n4.Array Index
    Item:"Retarget IKChain Settings Structure (by ref)"=@n7
    bSizeToFit:bool=false
   ) -> (then)
  ]}
]
variables [
 IKRetargeterAsset {type=IKRetargeter, default=&/IKR_SilentNovel_Internal}
 IKRetargetIKChainsController {type=IKRetargetIKChainsController, default=null}
 IKRetargetPelvisMotionController {type=IKRetargetPelvisMotionController, default=null}
 IKRetargetProfile {
  type=RetargetProfile
  default {
   bApplyChainSettings=false
   bApplyGlobalSettings=false
   bApplyRootSettings=false
   bApplySourceRetargetPose=false
   bApplyTargetRetargetPose=false
   bForceAllIKOff=false
   sourceRetargetPoseName=None
   targetRetargetPoseName=None
   chainSettings {}
   globalSettings {
    bCopyBasePose=false
    bEnableFK=true
    bEnableIK=true
    bEnablePost=true
    bEnableRoot=true
    bWarping=false
    copyBasePoseRoot=None
    directionChain=None
    directionSource=Goals
    forwardDirection=Y
    sidewaysOffset=0
    sourceScaleFactor=1
    warpForwards=1
    warpSplay=1
   }
   retargetOpProfiles []
   rootSettings {
    affectIKHorizontal=1
    affectIKVertical=0
    blendToSource=0
    rotationAlpha=1
    scaleHorizontal=1
    scaleVertical=1
    translationAlpha=1
    blendToSourceWeights=<1 1 1>
    rotationOffset=<rot 0 0 0>
    translationOffset=<0 0 0>
   }
  }
 }
 LeftArmIK {type=boolean}
 PelvisTranslationOffset {type=Vector}
 RightArmIK {type=boolean}
]
